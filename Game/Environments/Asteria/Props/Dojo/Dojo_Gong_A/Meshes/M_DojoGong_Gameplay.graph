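# Nanite-enabled gameplay material for the Dojo gong prop, with a procedural swing animation.
# Time since StartTime samples an animation curve atlas, driving a RotateAboutAxis around a pivot, masked by red vertex color.
# Custom primitive data sets swing direction and a debug mode that uses StartTime as the curve test value.
# Static switches add a second texture layer blended by UV1 and curve-driven flickering emissive.
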Material /Game/Environments/Asteria/Props/Dojo/Dojo_Gong_A/Meshes/M_DojoGong_Gameplay
graph {
 properties {
  bUsedWithNanite=true
  pixelDepthOffsetMode=PDOM_Legacy
  previewMesh=/Hermes/Props/SprocketPoppy/SprocketPoppy_Gong_A/Meshes/SM_SprocketPoppy_Gong_A
  usedShadingModels=Default Lit
 }
 nodes [
  n1 = NamedRerouteUsage()
  LinearInterpolate(A=Constant3Vector() {constant=<rgba 0.5429 1 0.3397 1>}, B=Constant3Vector() {constant=<rgba 1 0 0.0384 1>}, Alpha=@n1)
  MaterialFunctionCall(Float3=NamedRerouteUsage()) {materialFunction=&BreakOutFloat3Components}
  n2 = Transform(None=VectorParameter() {parameterName=Axis, defaultValue=<rgba 1 0 0 1>}.RGB) {transformSourceType=TRANSFORMSOURCE_Local}
  n3 = Multiply(A=ScalarParameter() {defaultValue=0.1, parameterName=RotateAmount}, B=NamedRerouteUsage())
  n4 = LinearInterpolate(Alpha=ScalarParameter() {defaultValue=1, parameterName=IsForward, primitiveDataIndex=2}) {constA=-1}
  n5 = VectorParameter() {parameterName=Pivot, defaultValue=<rgba 0 0 204 1>}
  Multiply(
   A=RotateAboutAxis(
    NormalizedRotationAxis=Normalize(VectorInput=@n2)
    RotationAngle=Multiply(A=@n3, B=@n4)
    PivotPoint=TransformPosition(None=@n5.RGB) {transformType=TRANSFORMPOSSOURCE_World}
    Position=WorldPosition().XYZ
   )
   B=VertexColor().R
  )
  n6 = TextureSampleParameter2D() {parameterName=EmissiveMask, samplerType=SAMPLERTYPE_Masks, texture=&T_BlackMask}
  n7 = Saturate(None=TextureCoordinate() {coordinateIndex=1})
  n8 = Multiply(
   A=Ceil(None=MaterialFunctionCall(Float2=@n7) {materialFunction=&BreakOutFloat2Components}.R)
   B=VectorParameter() {parameterName=Emissive_Steady, defaultValue=<rgba 0 0 0 1>}.RGB
  )
  n9 = Multiply(A=@n6.G, B=@n8)
  n10 = Time()
  n11 = ScalarParameter() {defaultValue=0.1, parameterName=FlickerSpeed}
  n12 = CurveAtlasRowParameter(CurveTime=Frac(None=Multiply(A=@n10, B=@n11))) {
   atlas=/Game/Environments/Apollo/Sets/Shared/Props/Textures/CA_Shared_LightFlickers
   curve=/Game/Environments/Apollo/Sets/Shared/Props/Textures/CC_Shared_FlickerCurve_03
   defaultValue=2
   parameterName=FlickerRate
  }
  n13 = Multiply(A=@n6.B, B=Multiply(A=VectorParameter() {parameterName=EmissiveFlicker, defaultValue=<rgba 5 0 0 1>}.RGB, B=@n12))
  Multiply(A=NamedRerouteUsage(), B=StaticSwitchParameter(True=Add(A=@n9, B=@n13), False=@n8) {parameterName=UseFlickerLighting})
  n14 = ScalarParameter() {defaultValue=-1, parameterName=StartTime}
  n15 = Multiply(A=Subtract(A=Time() {period=0.0001}, B=@n14), B=ScalarParameter() {defaultValue=0.5, parameterName=AnimSpeed})
  n16 = CurveAtlasRowParameter(CurveTime=LinearInterpolate(
    A=If(A=@n14, "A > B"=@n15, "A == B"=@n14, "A < B"=@n14)
    B=ScalarParameter() {defaultValue=-1, parameterName=StartTime}
    Alpha=ScalarParameter() {parameterName=DEBUG_UseStartTimeAsTestValue, primitiveDataIndex=1}
   )) {
   atlas=&/Atlas_DojoGong_Anim
   curve=&/Curve_DojoGong_Anim
   parameterName=AnimCurve
  }
  NamedRerouteDeclaration(None=Subtract(A=@n16.R) {constB=0.5}) {name=TimeVal, nodeColor=<rgba 0.3125 0 1 1>}
  n17 = TextureSampleParameter2D() {parameterName=SpecularMasks, samplerType=SAMPLERTYPE_Masks, texture=&T_BlackMask}
  n18 = TextureSampleParameter2D() {group=Layer2, parameterName=SpecularMasks_Texture_2, samplerType=SAMPLERTYPE_Masks, texture=&T_BlackMask}
  n19 = NamedRerouteUsage()
  NamedRerouteDeclaration(None=StaticSwitchParameter(True=LinearInterpolate(A=@n17.RGB, B=@n18.RGB, Alpha=@n19), False=@n17.RGB) {group=Layer2, parameterName=Use2Layers}) {name=SpecMask_Route, nodeColor=<rgba 1 0 0.5156 1>}
  n20 = TextureSampleParameter2D() {parameterName=Diffuse, texture=/Game/Global/Textures/Default/Blanks/T_Blank_LegalWhite_sRGBColor}
  n21 = TextureSampleParameter2D() {group=Layer2, parameterName=Diffuse_Texture_2, texture=&T_Blank_LegalWhite_sRGBColor}
  NamedRerouteDeclaration(None=StaticSwitchParameter(True=LinearInterpolate(A=@n20.RGB, B=@n21.RGB, Alpha=@n1), False=@n20.RGB) {group=Layer2, parameterName=Use2Layers}) {name=Diffuse_Route, nodeColor=<rgba 0 0.8359 1 1>}
  n22 = TextureSampleParameter2D() {parameterName=Normals, samplerType=SAMPLERTYPE_Normal, texture=&FlatNormalBlue}
  n23 = TextureSampleParameter2D() {group=Layer2, parameterName=Normals_Texture_2, samplerType=SAMPLERTYPE_Normal, texture=&FlatNormalBlue}
  n24 = NamedRerouteUsage()
  NamedRerouteDeclaration(None=StaticSwitchParameter(True=LinearInterpolate(A=@n22.RGB, B=@n23.RGB, Alpha=@n24), False=@n22.RGB) {group=Layer2, parameterName=Use2Layers}) {name=Normal_Route, nodeColor=<rgba 1 0.8438 0 1>}
  n25 = ComponentMask(None=TextureCoordinate() {coordinateIndex=1}) {r=true}
  NamedRerouteDeclaration(None=Ceil(None=Subtract(A=@n25))) {name=Layer2Blend_Route, nodeColor=<rgba 0.5234 0 1 1>}
  NamedRerouteUsage()
 ]
}
